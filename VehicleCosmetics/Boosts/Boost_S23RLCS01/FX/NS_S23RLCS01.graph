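# Niagara vehicle-cosmetic boost effect for S23RLCS01: a layered exhaust ribbon trail with a glow at the boost position.
# The ribbon emitter offsets from user position variables, sets backward velocity by game mode and fades inherited source movement over lifetime.
# Seven ribbon renderers share those particles, with per-layer blue, orange and yellow color and curve-driven width attributes.
# A second emitter bursts two persistent, camera-offset warm glow sprites; painted-variant overrides are disabled in both.
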NiagaraSystem /VehicleCosmetics/Boosts/Boost_S23RLCS01/FX/NS_S23RLCS01
dependencies {
 usedDataInterfaces [&NiagaraDataInterfaceCurve, &NiagaraDataInterfaceSkeletalMesh, &NiagaraDataInterfaceArrayFloat3]
 usedDynamicInputs [&SelectVectorFromArray]
 usedModules [
  &self:SystemSpawnScript.SystemScriptSource.SystemScriptGraph.NiagaraNodeAssignment_0.SetVariables_#1
  &self:SystemSpawnScript.SystemScriptSource.SystemScriptGraph.NiagaraNodeAssignment_1.SetVariables_#2
  &self:SystemSpawnScript.SystemScriptSource.SystemScriptGraph.NiagaraNodeAssignment_2.SetVariables_#3
  &SystemState
 ]
}
emitters {
 NE_VC_SingleBoost {
  emitterUpdateScript [
   &EmitterState(Inactive Response=::NewEnumerator1, Life Cycle Mode=::NewEnumerator1, Loop Duration=1, UseLoopDelay=false)
   &VC_FilteredSocketCount()
   &VC_RateQualityHandler(High Spawn Rate=40, Low Spawn Rate=20, Use Custom Rate?=true)
   &VC_SpawnRibbon_v1(SpawnRate=$Emitter.Spawn Rate)
  ]
  particleSpawnScript [
   &VC_PaintedVariantsOverride() {enabled=false}
   &VC_InitializeRibbon_v1(
    Color=<rgba 0.85 0.85 0.85 1>
    Custom Offset=true
    Lifetime=0.8
    Local Vector Offset=&AddVectorToPosition(
     Position=<0 0 0>
     Vector=&AddVectorToPosition(Position=$User.SingleBoostPos, Vector=$User.DTCarOffset)
    )
    Offset Amount=1
    Ribbon Width=60
   )
   &AddVelocity(Velocity=&MergeHelper_VectorSwitch(0=<-1800 0 0>, 1=<-1500 0 0>, 2=<-1500 0 0>, 3=<-1800 0 0>, Index=$User.GameModeID, Num=::NewEnumerator2)) {module=AddVelocity001}
   /VehicleCosmetics/FX/VC_FXParents/Modules/DLR_ProductPreviews(GameModeID:NiagaraInt32=$System.GameModeID, bPreviewVehicleSpeed:NiagaraBool=$System.bPreviewVehicleSpeed)
   /SetVariables(
    Particles.blue:LinearColor=&MultiplyLinearColor(
     Linear Color=<rgba 0.0012 0.0012 0.06 1>
     Scale Factor=&MakeVector4FromVectorAndFloat("Vector (RGB)"=&VectorFromFloat(Value=1), "Float (Alpha)"=1)
    )
    Particles.darkblue:LinearColor=&MultiplyLinearColor(
     Linear Color=<rgba 0.01 0.01 0.04 1>
     Scale Factor=&MakeVector4FromVectorAndFloat("Vector (RGB)"=&VectorFromFloat(Value=1), "Float (Alpha)"=1)
    )
    Particles.orange:LinearColor=&MultiplyLinearColor(
     Linear Color=<rgba 0.8 0.3408 0.128 1>
     Scale Factor=&MakeVector4FromVectorAndFloat("Vector (RGB)"=&VectorFromFloat(Value=1), "Float (Alpha)"=2)
    )
    Particles.yellow:LinearColor=&MultiplyLinearColor(
     Linear Color=<rgba 0.35 0.2975 0.035 1>
     Scale Factor=&MakeVector4FromVectorAndFloat("Vector (RGB)"=&VectorFromFloat(Value=2), "Float (Alpha)"=1)
    )
   ) {isSetParameters=true}
  ]
  particleUpdateScript [
   &ParticleState()
   &InheritSourceMovement(
    Applied Position Scale=&VectorFromFloat(Value=&FloatFromCurve(FloatCurve=DataInterface() {propertyValues {Curve {
         defaultValue=1
         keys [
          {arriveTangent=-1.7885, interpMode=RCIM_Cubic, leaveTangent=-1.7885, tangentMode=RCTM_User, time=0.05, value=1}
          {interpMode=RCIM_Cubic, tangentMode=RCTM_User, time=1, value=0}
         ]
        }}}))
    ApplyForce=false
    ApplyPosition=true
    ApplyPositionFalloff=false
   ) {module=InheritSourceMovement001}
   &SolveForcesAndVelocity(Clamp Velocity=false, Limit Acceleration=false)
   &VC_UpdateRibbon_v1()
   &ScaleRibbonWidth(Ribbon Width Scale=&Multiply_Float(
     A=&FloatFromCurve(FloatCurve=DataInterface() {propertyValues {Curve {keys [
          {interpMode=RCIM_Cubic, leaveTangentWeight=0.1333, tangentMode=RCTM_User, time=0, value=0}
          {arriveTangentWeight=0.1436, interpMode=RCIM_Cubic, leaveTangentWeight=0.2154, tangentMode=RCTM_User, time=0.4, value=1}
          {arriveTangentWeight=0.2, interpMode=RCIM_Cubic, tangentMode=RCTM_User, time=1, value=0.4}
         ]}}})
     B=1.6
    ))
   /SetVariables(
    Particles.colors:NiagaraFloat=&Multiply_Float(
     A=&FloatFromCurve(FloatCurve=DataInterface() {propertyValues {Curve {keys [
          {interpMode=RCIM_Cubic, leaveTangentWeight=0.1333, tangentMode=RCTM_User, time=0, value=0}
          {arriveTangentWeight=0.1436, interpMode=RCIM_Cubic, leaveTangentWeight=0.2154, tangentMode=RCTM_User, time=0.4, value=1}
          {arriveTangentWeight=0.2, interpMode=RCIM_Cubic, tangentMode=RCTM_User, time=1, value=0.4}
         ]}}})
     B=45
    )
    Particles.widthbluecolors:NiagaraFloat=&Multiply_Float(
     A=&FloatFromCurve(FloatCurve=DataInterface() {propertyValues {Curve {keys [
          {interpMode=RCIM_Cubic, leaveTangentWeight=0.0667, tangentMode=RCTM_User, time=0, value=0.1}
          {
           arriveTangent=7.9544
           arriveTangentWeight=0.0722
           interpMode=RCIM_Cubic
           leaveTangentWeight=0.2611
           tangentMode=RCTM_Break
           time=0.2
           value=1
          }
          {arriveTangent=-1, arriveTangentWeight=0.2611, interpMode=RCIM_Cubic, leaveTangent=-1, tangentMode=RCTM_User, time=1, value=0.4}
         ]}}})
     B=125
    )
    Particles.widthdarkbluecolors:NiagaraFloat=&Multiply_Float(
     A=&FloatFromCurve(FloatCurve=DataInterface() {propertyValues {Curve {keys [
          {arriveTangent=3.9109, interpMode=RCIM_Cubic, leaveTangent=3.9109, leaveTangentWeight=0.0667, tangentMode=RCTM_User, time=0, value=0.4}
          {
           arriveTangent=1.292
           arriveTangentWeight=0.067
           interpMode=RCIM_Cubic
           leaveTangent=1.292
           leaveTangentWeight=0.268
           tangentMode=RCTM_User
           time=0.3
           value=1.2
          }
          {arriveTangent=-1.0333, arriveTangentWeight=0.2494, interpMode=RCIM_Cubic, leaveTangent=-1.0333, tangentMode=RCTM_User, time=1, value=0.2}
         ]}}})
     B=60
    )
    Particles.widthorange:NiagaraFloat=&Multiply_Float(
     A=&FloatFromCurve(FloatCurve=DataInterface() {propertyValues {Curve {keys [
          {arriveTangent=10.4848, interpMode=RCIM_Cubic, leaveTangent=10.4848, leaveTangentWeight=0.0667, tangentMode=RCTM_User, time=0, value=0}
          {
           arriveTangent=0.3636
           arriveTangentWeight=0.0722
           interpMode=RCIM_Cubic
           leaveTangent=0.2656
           leaveTangentWeight=0.2611
           tangentMode=RCTM_Break
           time=0.2
           value=1
          }
          {arriveTangent=-1.2803, arriveTangentWeight=0.2611, interpMode=RCIM_Cubic, leaveTangent=-1.2803, tangentMode=RCTM_User, time=1, value=0.4}
         ]}}})
     B=70
    )
    Particles.widthyellow:NiagaraFloat=&Multiply_Float(
     A=&FloatFromCurve(FloatCurve=DataInterface() {propertyValues {Curve {keys [
          {arriveTangent=5.9904, interpMode=RCIM_Cubic, leaveTangent=5.9904, leaveTangentWeight=0.0667, tangentMode=RCTM_User, time=0, value=0.2}
          {arriveTangentWeight=0.0722, interpMode=RCIM_Cubic, leaveTangentWeight=0.2611, time=0.2, value=0.8}
          {arriveTangentWeight=0.2667, interpMode=RCIM_Cubic, tangentMode=RCTM_User, time=1, value=0.2}
         ]}}})
     B=125
    )
   ) {isSetParameters=true}
  ]
  rendererClasses [&NiagaraRibbonRendererProperties]
  renderers [
   {rendererClass=&NiagaraRibbonRendererProperties}
   {rendererClass=&NiagaraRibbonRendererProperties, rendererIndex=1}
   {rendererClass=&NiagaraRibbonRendererProperties, rendererIndex=2}
   {rendererClass=&NiagaraRibbonRendererProperties, rendererIndex=3}
   {rendererClass=&NiagaraRibbonRendererProperties, rendererIndex=4}
   {rendererClass=&NiagaraRibbonRendererProperties, rendererIndex=5}
   {rendererClass=&NiagaraRibbonRendererProperties, rendererIndex=6}
  ]
 }
 NE_VC_SingleBoost_Sprite_Parent {
  emitterUpdateScript [
   &EmitterState(
    Inactive Response=::NewEnumerator1
    Life Cycle Mode=::NewEnumerator1
    Loop Behavior=::NewEnumerator1
    Loop Duration Mode=::NewEnumerator1
    UseLoopDelay=false
   )
   &SpawnBurst_Instantaneous(Spawn Count=2, Use Spawn Probability=false, UseLoopCountLimit=false)
   &SpawnRate(SpawnRate=50, Use Spawn Probability=false) {enabled=false}
  ]
  particleSpawnScript [
   &VC_PaintedVariantsOverride() {enabled=false}
   &InitializeParticle(
    Color=<rgba 5 2.6 0.2 1>
    Position Offset=<-200 0 30>
    Position Offset Coordinate Space=::Local
    Sprite Size=<80 50>
    Sprite Size Mode=::NewEnumerator4
    UsePositionOffset=true
    Write Material Random=true
   )
   &AddVelocity(Velocity=<-1500 0 0>) {enabled=false}
   /SetVariables(
    Particles.glowcolor:LinearColor=<rgba 3 0.625 0.15 0.5>
    Particles.glowsize:Vector2f=&Multiply_Vector2D_ByFloat(Float=1, Vector2D=<300 60>)
   ) {isSetParameters=true}
  ]
  particleUpdateScript [
   &ParticleState(Kill Particles When Lifetime Has Elapsed=false)
   &SolveForcesAndVelocity(Clamp Velocity=false, Limit Acceleration=false)
   &CameraOffset(Camera Offset Amount=80)
  ]
  rendererClasses [&NiagaraSpriteRendererProperties]
  renderers [
   {rendererClass=&NiagaraSpriteRendererProperties}
   {rendererClass=&NiagaraSpriteRendererProperties, rendererIndex=1}
  ]
 }
}
summary {
 systemName=NS_S23RLCS01
 emitters [
  {emitterName=NE_VC_SingleBoost, rendererClasses [&NiagaraRibbonRendererProperties]}
  {emitterName=NE_VC_SingleBoost_Sprite_Parent, rendererClasses [&NiagaraSpriteRendererProperties]}
 ]
 userVariables [
  User.bIsFrontendPreview {type=NiagaraBool}
  User.bIsPainted {type=NiagaraBool}
  User.bPreviewVehicleSpeed {type=NiagaraBool}
  User.ColorVariantID {type=NiagaraInt32}
  User.DTCarOffset {type=Vector3f, description="", defaultValue {struct=Vector3f, value=<0 0 0>}}
  User.GameModeID {type=NiagaraInt32}
  User.PaintedColor {type=LinearColor, description="", defaultValue {struct=LinearColor, value=<rgba 1 1 1 1>}}
  User.PaintedScalar {type=NiagaraFloat, defaultValue=1, description=""}
  User.PortCount {type=NiagaraInt32, defaultValue=1}
  User.SingleBoostPos {type=NiagaraPosition, description="", defaultValue {}}
 ]
}
